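# Masked, lit particle material for ice flicks on sprite and Niagara sprite particles, with time-of-day lighting response.
# Each particle picks a random cell from a 2x2 ice mask, using fixed scaled UVs on mobile; mask times particle alpha feeds dithered temporal AA opacity.
# Emissive adds Fresnel tint to fake crystal reflections, sampled by view-space reflection vector and shaped by a water normal map.
# Tints, reflection textures and intensity vary by time of day.
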Material /Game/Effects/Fort_Effects/World/Enviroment/Snow/M_IceFlicks_GPU
graph {
 properties {
  bUsedWithNiagaraSprites=true
  bUsedWithParticleSprites=true
  blendMode=BLEND_Masked
  pixelDepthOffsetMode=PDOM_Legacy
  usedShadingModels=Default Lit
 }
 nodes [
  n1 = Constant() {r=2}
  n2 = ParticleRandom()
  n3 = TextureSample(UVs=ShadingPathSwitch(
    Default=MaterialFunctionCall!(Rows=@n1, Columns=@n1, Particle Random Value=@n2) {materialFunction=/Game/Effects/Fort_Effects/Materials/_Functions/RandomSubUV_GPUSprites}.Result
    Mobile=TextureCoordinate() {uTiling=0.25, vTiling=0.25}
   ) {desc=Migrated from FeatureLevelSwitch}) {texture=/Game/Weapons/WeaponSkins/Elemental/Textures/IceMasks}
  LinearInterpolate(
   A=Constant3Vector() {constant=<rgba 0 0 1 0>}
   B=TextureSample() {samplerType=SAMPLERTYPE_Normal, texture=/Game/Packages/Fortnite_UniqueMaterials_1/Textures/water2_NRM}.RGB
   Alpha=@n3.RGB
  )
  MaterialFunctionCall(Alpha Threshold=Multiply(A=ParticleColor().A, B=@n3.RGB) {desc=opacity}) {materialFunction=&DitherTemporalAA_Fortnite}
  n4 = ScalarParameter() {defaultValue=0.25, parameterName=Emissive Intensity_Morning}
  n5 = ScalarParameter() {defaultValue=1, parameterName=Emissive Intensity_Day}
  n6 = ScalarParameter() {defaultValue=1, parameterName=Emissive Intensity_Evening}
  n7 = ScalarParameter() {defaultValue=0.25, parameterName=Emissive Intensity_Night}
  MaterialFunctionCall(Color_Morning=@n4, Color_Day=@n5, Color_Evening=@n6, Color_Night=@n7) {materialFunction=&Color_Time_of_Day}
  n8 = Fresnel() {baseReflectFraction=0.25, exponent=1}
  n9 = OneMinus(None=@n8)
  MaterialFunctionCall(SpecularColor=Multiply(A=Fresnel() {baseReflectFraction=0, exponent=1}, B=@n9)) {materialFunction=&ConvertFromDiffSpec}
  n10 = Multiply(
   A=MaterialFunctionCall(
    Color_Morning=VectorParameter() {parameterName=Color Morning, defaultValue=<rgba 1 0.6506 0.3424 0>}.RGB
    Color_Day=VectorParameter() {parameterName=Color Day, defaultValue=<rgba 0.3004 0.4266 0.5 0>}.RGB
    Color_Evening=VectorParameter() {parameterName=Color Evening, defaultValue=<rgba 1.3 0.857 0.533 0>}.RGB
    Color_Night=VectorParameter() {parameterName=Color Night, defaultValue=<rgba 0.27 0.2723 1 0>}.RGB
   ) {materialFunction=&Color_Time_of_Day}.Result
   B=Clamp(None=@n8)
  )
  n11 = MaterialFunctionCall(Float3=Transform(None=ReflectionVectorWS()) {transformSourceType=TRANSFORMSOURCE_World, transformType=TRANSFORM_View}) {materialFunction=&BreakOutFloat3Components}
  n12 = Multiply(A=@n11.G) {constB=-1}
  n13 = Constant() {r=0.5}
  n14 = Add(A=Multiply(A=AppendVector(A=@n11.R, B=@n12), B=@n13), B=@n13)
  n15 = TextureSample(UVs=@n14) {texture=/Game/Environments/Props/Rocks/Textures/T_Props_CrystalEnv_White}
  n16 = TextureSample(UVs=@n14) {texture=/Game/Environments/Props/Rocks/Textures/T_Props_CrystalEnv_Purple}
  n17 = TextureSample(UVs=@n14) {texture=/Game/Environments/Props/Rocks/Textures/T_Props_CrystalEnv_Red}
  n18 = TextureSample(UVs=@n14) {texture=/Game/Environments/Props/Rocks/Textures/T_Props_CrystalEnv_Green}
  n19 = Constant3Vector() {constant=<rgba 0 0 1 0>}
  n20 = TextureSample() {samplerType=SAMPLERTYPE_Normal, texture=&water2_NRM}
  n21 = Ceil(None=ComponentMask(None=TextureCoordinate()) {g=true})
  n22 = MaterialFunctionCall(f2=Abs(None=LinearInterpolate(A=@n19, B=@n20.RGB, Alpha=@n21))) {materialFunction=&AddComponents}
  n23 = Multiply(
   A=Multiply(A=MaterialFunctionCall(Color_Morning=@n15.RGB, Color_Day=@n16.RGB, Color_Evening=@n17.RGB, Color_Night=@n18.RGB) {materialFunction=&Color_Time_of_Day}.Result) {constB=5}
   B=Multiply(A=Power(Base=@n22.f2))
  )
  MaterialFunctionCall(Emissive=Multiply(
    A=Add(A=@n10, B=@n23)
    B=MaterialFunctionCall(Value_Morning=@n4, Value_Day=@n5, Value_Evening=@n6, Value_Night=@n7) {materialFunction=&Scalar_Time_of_Day}.Result
   ) {constB=1.5}) {materialFunction=&MF_FX_EmissiveAdjustments}
  Multiply(B=Multiply(A=@n9, B=@n9)) {constA=0.5, constB=0}
 ]
}
