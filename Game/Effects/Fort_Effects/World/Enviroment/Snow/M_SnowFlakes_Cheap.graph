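# Unlit translucent particle material for snowflake sprites, with color, reflection and emissive intensity that follow time of day.
# Fresnel against a blended sun/moon and camera vector drives cel-banded tints, plus faked view-space environment reflections masked by flake normals.
# Particle random selects among three channel-packed flake normals and shapes.
# Opacity blends sharp and blurred flakes by depth around a focus distance, reduced by particle speed, times particle alpha.
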Material /Game/Effects/Fort_Effects/World/Enviroment/Snow/M_SnowFlakes_Cheap
graph {
 properties {
  bUsedWithParticleSprites=true
  blendMode=BLEND_Translucent
  pixelDepthOffsetMode=PDOM_Legacy
  shadingModel=MSM_Unlit
  translucencyLightingMode=TLM_VolumetricPerVertexDirectional
  twoSided=true
  usedShadingModels=Unlit
 }
 nodes [
  n1 = MaterialFunctionCall(
   Color_Morning=VectorParameter() {group=Morning Smoke, parameterName=Morning Darks Smoke, defaultValue=<rgba 0.2783 0.1245 0.0779 0>}.RGB
   Color_Day=VectorParameter() {group=Daytime Smoke, parameterName=Daytime Darks Smoke, defaultValue=<rgba 0.1743 0.2455 0.3206 0>}.RGB
   Color_Evening=VectorParameter() {group=Evening Smoke, parameterName=Evening Darks Smoke, defaultValue=<rgba 0.41 0.2237 0.1722 0>}.RGB
   Color_Night=VectorParameter() {group=Nighttime Smoke, parameterName=Nighttime Darks Smoke, defaultValue=<rgba 0.0614 0.1056 0.2117 0>}.RGB
  ) {desc=Darks, materialFunction=&Color_Time_of_Day}
  n2 = MaterialFunctionCall(
   Color_Morning=VectorParameter() {group=Morning Smoke, parameterName=Morning Mids Smoke, defaultValue=<rgba 0.445 0.2656 0.155 0>}.RGB
   Color_Day=VectorParameter() {group=Daytime Smoke, parameterName=Daytime Mids Smoke, defaultValue=<rgba 0.2041 0.2938 0.4183 0>}.RGB
   Color_Evening=VectorParameter() {group=Evening Smoke, parameterName=Evening Mids Smoke, defaultValue=<rgba 0.63 0.2681 0.1984 0>}.RGB
   Color_Night=VectorParameter() {group=Nighttime Smoke, parameterName=Nighttime Mids Smoke, defaultValue=<rgba 0.1213 0.1645 0.2939 0>}.RGB
  ) {desc=Mids, materialFunction=&Color_Time_of_Day}
  n3 = VectorParameter() {group=Morning Smoke, parameterName=Morning Highlights Smoke_1, defaultValue=<rgba 0.7183 0.5222 0.3663 0>}
  n4 = VectorParameter() {group=Evening Smoke, parameterName=Evening Highlights Smoke_1, defaultValue=<rgba 0.64 0.4203 0.3648 0>}
  n5 = VectorParameter() {group=Nighttime Smoke, parameterName=Nighttime Highlights Smoke_1, defaultValue=<rgba 0.1955 0.4091 0.4894 0>}
  n6 = MaterialFunctionCall(
   Color_Morning=@n3.RGB
   Color_Day=VectorParameter() {group=Daytime Smoke, parameterName=Daytime Highlights Smoke_1, defaultValue=<rgba 0.3103 0.3962 0.6 0>}.RGB
   Color_Evening=@n4.RGB
   Color_Night=@n5.RGB
  ) {desc=Highlights, materialFunction=&Color_Time_of_Day}
  n7 = CollectionParameter() {
   collection=&FortniteMaterialParameters
   parameterName=SunAndMoonModelDirectionalVector
  }
  n8 = LinearInterpolate(
   A=ComponentMask(None=@n7) {b=true, g=true, r=true}
   B=CameraVectorWS()
   Alpha=ScalarParameter() {defaultValue=0.6, group=Cell Shading Parameters, parameterName=Sun/Moon Directional Shading Blend Smoke}
  )
  n9 = Add(
   A=MaterialFunctionCall(Camera Vector=Normalize(VectorInput=@n8), Vertex / Pixel Normal=PixelNormalWS()) {materialFunction=&MF_Fresnel_FX}.Fresnel
   B=ConstantBiasScale(None=ParticleColor().A) {bias=-1}
  )
  n10 = MaterialFunctionCall(1=@n1.Result, 2=@n2.Result, 3=@n6.Result, A=@n9) {materialFunction=&Lerp_Multiple_Float3}
  n11 = Multiply(A=@n9, B=ScalarParameter() {defaultValue=2, desc="Starts at 0 (so 2 = 3 steps)", group=Morning Smoke, parameterName=Cell Shading Steps})
  n12 = Add(
   A=@n11
   B=MaterialFunctionCall(
    Color_Morning=ScalarParameter() {defaultValue=0.3, group=Morning Smoke, parameterName=Cell Shaded Dark/Mid Bias Morning Smoke}
    Color_Day=ScalarParameter() {defaultValue=0.25, group=Daytime Smoke, parameterName=Cell Shaded Dark/Mid Bias Day Smoke}
    Color_Evening=ScalarParameter() {group=Evening Smoke, parameterName=Cell Shaded Dark/Mid Bias Evening Smoke}
    Color_Night=ScalarParameter() {defaultValue=0.49, group=Nighttime Smoke, parameterName=Cell Shaded Dark/Mid Bias Night Smoke}
   ) {materialFunction=&Color_Time_of_Day}.Result
  )
  n13 = Clamp(None=Multiply(A=Round(None=@n12)) {constB=2})
  n14 = Add(
   A=@n11
   B=MaterialFunctionCall(
    Color_Morning=ScalarParameter() {defaultValue=0.6, group=Morning Smoke, parameterName=Cell Shaded Mid/Light Bias Morning Smoke}
    Color_Day=ScalarParameter() {defaultValue=0.5, group=Daytime Smoke, parameterName=Cell Shaded Mid/Light Bias Day Smoke}
    Color_Evening=ScalarParameter() {defaultValue=0.5, group=Evening Smoke, parameterName=Cell Shaded Mid/Light Bias Evening Smoke}
    Color_Night=ScalarParameter() {defaultValue=0.9, group=Nighttime Smoke, parameterName=Cell Shaded Mid/Light Bias Night Smoke}
   ) {materialFunction=&Color_Time_of_Day}.Result
  )
  n15 = Clamp(None=ConstantBiasScale(None=Round(None=@n14)) {bias=-1, scale=2})
  n16 = LinearInterpolate(A=LinearInterpolate(A=@n1.Result, B=@n2.Result, Alpha=@n13) {desc=Dark / Mid Lerp}, B=@n6.Result, Alpha=@n15) {desc=Mid / Light Lerp}
  n17 = ScalarParameter() {defaultValue=0.4, parameterName=Cell Shading LERP}
  n18 = MaterialFunctionCall(Float3=Transform(None=ReflectionVectorWS()) {transformSourceType=TRANSFORMSOURCE_World, transformType=TRANSFORM_View}) {materialFunction=&BreakOutFloat3Components}
  n19 = Multiply(A=@n18.G) {constB=-1}
  n20 = Constant() {r=0.5}
  n21 = Add(A=Multiply(A=AppendVector(A=@n18.R, B=@n19), B=@n20), B=@n20)
  n22 = TextureSample(UVs=@n21) {texture=/Game/Environments/Props/Rocks/Textures/T_Props_CrystalEnv_White}
  n23 = ScalarParameter() {defaultValue=0.25, parameterName=Reflection Intensity Dark}
  n24 = TextureSample(UVs=@n21) {texture=/Game/Environments/Props/Rocks/Textures/T_Props_CrystalEnv_Purple}
  n25 = TextureSample(UVs=@n21) {texture=/Game/Environments/Props/Rocks/Textures/T_Props_CrystalEnv_Red}
  n26 = TextureSample(UVs=@n21) {texture=/Game/Environments/Props/Rocks/Textures/T_Props_CrystalEnv_Green}
  n27 = MaterialFunctionCall(
   Color_Morning=Multiply(A=@n22.RGB, B=@n23) {constB=1.5}
   Color_Day=Multiply(A=@n24.RGB, B=@n23)
   Color_Evening=Multiply(A=@n25.RGB, B=@n23)
   Color_Night=Multiply(A=@n26.RGB, B=@n23)
  ) {materialFunction=&Color_Time_of_Day}
  n28 = MaterialFunctionCall(Color_Morning=@n22.RGB, Color_Day=@n24.RGB, Color_Evening=@n25.RGB, Color_Night=@n26.RGB) {materialFunction=&Color_Time_of_Day}
  n29 = ScalarParameter() {defaultValue=1.5, parameterName=Reflection Intensity Highlight}
  n30 = MaterialFunctionCall(
   Color_Morning=Multiply(A=@n22.RGB, B=@n29)
   Color_Day=Multiply(A=@n24.RGB, B=@n29)
   Color_Evening=Multiply(A=@n25.RGB, B=@n29)
   Color_Night=Multiply(A=@n26.RGB, B=@n29)
  ) {materialFunction=&Color_Time_of_Day}
  n31 = LinearInterpolate(
   A=MaterialFunctionCall(1=@n27.Result, 2=@n28.Result, 3=@n30.Result) {materialFunction=&Lerp_Multiple_Float3}.Lerp 3 Inputs
   B=LinearInterpolate(A=LinearInterpolate(A=@n27.Result, B=@n28.Result, Alpha=@n13) {desc=Dark / Mid Lerp}, B=@n30.Result, Alpha=@n15) {desc=Mid / Light Lerp}
   Alpha=ScalarParameter() {defaultValue=1, parameterName=Cell Shading LERP_1}
  )
  n32 = ScalarParameter() {defaultValue=2, parameterName=ToD_Reflection_Intensity}
  n33 = Constant3Vector() {constant=<rgba 0 0 1 0>}
  n34 = ParticleRandom()
  n35 = Frac(None=@n34)
  n36 = TextureSample() {samplerType=SAMPLERTYPE_Normal, texture=/Game/Weapons/WeaponSkins/Elemental/Textures/T_Snow_01_Blue_N}
  n37 = TextureSample() {samplerType=SAMPLERTYPE_Normal, texture=/Game/Weapons/WeaponSkins/Elemental/Textures/T_Snow_01_Red_N}
  n38 = If(
   A=@n35
   B=Constant() {r=0.66}
   "A > B"=@n36.RGB
   "A == B"=@n36.RGB
   "A < B"=If(
    A=@n35
    "A > B"=TextureSample() {samplerType=SAMPLERTYPE_Normal, texture=/Game/Weapons/WeaponSkins/Elemental/Textures/T_Snow_01_Green_N}.RGB
    "A == B"=@n37.RGB
    "A < B"=@n37.RGB
   )
  )
  n39 = Ceil(None=ComponentMask(None=TextureCoordinate()) {g=true})
  n40 = MaterialFunctionCall(f2=Abs(None=LinearInterpolate(A=@n33, B=@n38, Alpha=@n39))) {materialFunction=&AddComponents}
  n41 = Add(
   A=Multiply(A=LinearInterpolate(A=@n10.Lerp 3 Inputs, B=@n16, Alpha=@n17), B=ScalarParameter() {defaultValue=1, parameterName=ToD_Injection_Intensity})
   B=Multiply(A=Multiply(A=@n31, B=@n32) {constB=5}, B=Multiply(A=@n40.f2, B=@n40.f2))
  )
  n42 = MaterialFunctionCall(
   Color_Morning=ScalarParameter() {defaultValue=0.25, parameterName=Emissive Intensity_Morning}
   Color_Day=ScalarParameter() {defaultValue=1, parameterName=Emissive Intensity_Day}
   Color_Evening=ScalarParameter() {defaultValue=1, parameterName=Emissive Intensity_Evening}
   Color_Night=ScalarParameter() {defaultValue=0.25, parameterName=Emissive Intensity_Night}
  ) {materialFunction=&Color_Time_of_Day}
  n43 = ParticleColor()
  MaterialFunctionCall(Emissive=Multiply(A=Multiply(A=@n41, B=@n42.Result), B=@n43.RGB)) {materialFunction=&MF_FX_EmissiveAdjustments}
  n44 = TextureSample() {texture=/Game/Weapons/WeaponSkins/Elemental/Textures/T_Snow_01}
  n45 = TextureSample() {texture=/Game/Weapons/WeaponSkins/Elemental/Textures/T_Snow_01_blur}
  n46 = ScalarParameter() {defaultValue=800, parameterName=NoBlurRange_Centered}
  n47 = PixelDepth()
  n48 = Constant()
  n49 = Subtract(
   A=SphereMask(A=PixelDepth().R, B=Constant(), Radius=@n46) {hardnessPercent=10}
   B=OneMinus(None=SphereMask(A=@n47.R, B=@n48, Radius=@n46) {hardnessPercent=10})
  )
  n50 = ParticleSpeed()
  n51 = ScalarParameter() {defaultValue=500, parameterName=SpeedBlurThreshold}
  n52 = Clamp(None=Subtract(A=Abs(None=@n49), B=Divide(A=@n50, B=@n51) {constB=400}))
  Multiply(
   A=@n43.A
   B=MaterialFunctionCall!(Random Picker=@n34, Texture Input=LinearInterpolate(A=@n44.RGB, B=@n45.RGB, Alpha=@n52), Option 1, Option 2, Option 3) {materialFunction=/Game/Effects/Fort_Effects/Materials/_Functions/MF_RGB_Random_01}.Result
  )
 ]
}
