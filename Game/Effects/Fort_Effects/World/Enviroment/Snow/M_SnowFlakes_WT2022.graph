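# Translucent unlit Niagara sprite material for snowflakes, tinted by time of day and sun direction, with an optional glint.
# Opacity combines a randomly picked flake channel, which becomes a soft dot at high particle speed, with camera fade, optional distance culling and a switchable player-indoors mask.
# A vertex offset chain shrinks sprites near the camera, factors in screen aspect ratio, and can throw masked flakes far below the world.
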Material /Game/Effects/Fort_Effects/World/Enviroment/Snow/M_SnowFlakes_WT2022
graph {
 properties {
  bEnableResponsiveAA=true
  bUsedWithNiagaraSprites=true
  blendMode=BLEND_Translucent
  pixelDepthOffsetMode=PDOM_Legacy
  shadingModel=MSM_Unlit
  translucencyLightingMode=TLM_VolumetricPerVertexDirectional
  twoSided=true
  usedShadingModels=Unlit
 }
 nodes [
  Constant()
  Constant() {r=1}
  n1 = DynamicParameter() {paramNames [RandomFlake,X,Y,Z]}
  n2 = AppendVector(A=@n1.X, B=@n1.X)
  n3 = Clamp(None=DotProduct(A=Constant3Vector() {constant=<rgba 0 0 1 0>}, B=AppendVector(A=@n2, B=@n1.Z)))
  n4 = ScalarParameter() {defaultValue=1, parameterName=GlintPower}
  n5 = StaticSwitchParameter(True=Multiply(A=Power(Base=@n3, Exp=@n4), B=ScalarParameter() {defaultValue=0.1, parameterName=GlintMult}), False=Constant()) {parameterName=UseGlint?}
  n6 = CollectionParameter() {
   collection=&FortniteMaterialParameters
   parameterName=SunAndMoonModelDirectionalVector
  }
  n7 = DotProduct(
   A=ComponentMask(None=@n6) {b=true, g=true, r=true}
   B=Transform(None=Constant3Vector() {constant=<rgba 0 0 1 0>}) {transformSourceType=TRANSFORMSOURCE_View}
  )
  n8 = VectorParameter() {group=TOD, parameterName="Color_Morning (V3)", defaultValue=<rgba 1 1 1 0>}
  n9 = VectorParameter() {group=TOD, parameterName="Color_Day (V3)", defaultValue=<rgba 1 1 1 0>}
  n10 = VectorParameter() {group=TOD, parameterName="Color_Evening (V3)", defaultValue=<rgba 1 1 1 0>}
  n11 = VectorParameter() {group=TOD, parameterName="Color_Night (V3)", defaultValue=<rgba 1 1 1 0>}
  n12 = Multiply(
   A=LinearInterpolate(
    A=ScalarParameter() {defaultValue=0.5, group=TOD, parameterName=ColorIntensity_SunSide_Mult}
    B=ScalarParameter() {defaultValue=0.35, group=TOD, parameterName=ColorIntensity_ShadowSide_Mult}
    Alpha=Clamp(None=@n7)
   )
   B=Multiply(
    A=MaterialFunctionCall(Color_Morning=@n8.RGB, Color_Day=@n9.RGB, Color_Evening=@n10.RGB, Color_Night=@n11.RGB) {materialFunction=&Color_Time_of_Day}.Result
    B=ScalarParameter() {defaultValue=1, parameterName=ColorIntensity}
   )
  )
  MaterialFunctionCall(Emissive=Multiply(A=Add(A=@n5, B=@n12), B=ParticleColor().RGB)) {materialFunction=&MF_FX_EmissiveAdjustments}
  n13 = Add(A=Clamp(None=@n7), B=ScalarParameter() {defaultValue=1, parameterName=OpacitySunMult})
  n14 = CollectionParameter() {collection=&FortniteMaterialParameters, parameterName=IsPlayerIndoors}
  n15 = StaticSwitchParameter(True=OneMinus(None=Clamp(None=@n14)), False=Constant() {r=1}) {
   defaultValue=true
   desc="You might want snow to come inside openings in the structure. In that case, toggle this off."
   parameterName=Player Indoors Check
  }
  n16 = DistanceCullFade()
  n17 = DynamicParameter() {paramNames [RandomFlake,X,Y,Z]}
  n18 = TextureSampleParameter2D() {group=Textures, parameterName=Flake Texture, sortPriority=0, texture=/Game/Weapons/WeaponSkins/Elemental/Textures/T_Snow_01}
  n19 = SphereMask(A=TextureCoordinate(), B=Constant2Vector() {g=0.5, r=0.5}, Radius=Constant() {r=0.45}, Hardness=Constant())
  n20 = ParticleSpeed()
  n21 = ScalarParameter() {defaultValue=300, parameterName=SpeedBlurThreshold}
  n22 = LinearInterpolate(A=Multiply(A=@n18.RGB, B=@n19), B=@n19, Alpha=Clamp(None=Subtract(A=@n20, B=@n21)))
  n23 = Clamp(None=MaterialFunctionCall(Distance=ScalarParameter() {defaultValue=300, parameterName=ScaleDown_Start}) {materialFunction=&Near_Camera_Fade}.Result)
  n24 = ShadingPathSwitch(Default=@n23, Mobile=Constant() {r=1}) {desc=Migrated from FeatureLevelSwitch}
  n25 = Constant() {r=1}
  n26 = Multiply(
   A=Multiply(
    A=ParticleColor().A
    B=MaterialFunctionCall!(Random Picker=@n17.RandomFlake, Texture Input=@n22, Option 1, Option 2, Option 3) {materialFunction=/Game/Effects/Fort_Effects/Materials/_Functions/MF_RGB_Random_01}.Result
   )
   B=ShadingPathSwitch(Default=QualitySwitch(Default=@n24, Low=@n25), Forward=Constant() {r=1})
  )
  n27 = Multiply(
   A=ScalarParameter() {defaultValue=1, parameterName=Opacity}
   B=StaticSwitchParameter(True=Multiply(A=@n16, B=@n26), False=@n26) {parameterName=UseDistanceCullFade}
  )
  n28 = Multiply(A=@n15, B=@n27)
  Multiply(
   A=Multiply(A=Multiply(A=@n13, B=@n28), B=Constant() {r=1})
   B=MaterialFunctionCall(
    Value_Morning=ScalarParameter() {defaultValue=1, parameterName="Value_Morning (S)"}
    Value_Day=ScalarParameter() {defaultValue=1, parameterName="Value_Day (S)"}
    Value_Evening=ScalarParameter() {defaultValue=1, parameterName="Value_Evening (S)"}
    Value_Night=ScalarParameter() {defaultValue=1, parameterName="Value_Night (S)"}
   ) {materialFunction=&Scalar_Time_of_Day}.Result
  )
  n29 = ParticleSpeed()
  n30 = ScalarParameter() {defaultValue=25, parameterName=SpeedFadeThreshold}
  Multiply(A=@n27, B=Clamp(None=Divide(A=@n29, B=@n30) {constB=400}))
  n31 = Power(Base=Constant() {r=1}, Exp=ScalarParameter() {defaultValue=25, parameterName=MaskPower})
  n32 = Ceil(None=Add(A=Multiply(A=@n15, B=@n31), B=ScalarParameter() {defaultValue=0.2, parameterName=SubtractWPOMask_Value}))
  n33 = LinearInterpolate(A=Subtract(A=ParticlePositionWS(), B=WorldPosition().XYZ), B=Constant(), Alpha=@n23)
  n34 = Constant() {r=1}
  n35 = ShadingPathSwitch(
   Default=QualitySwitch(Default=ShadingPathSwitch(Default=@n33, Mobile=@n34) {desc=Migrated from FeatureLevelSwitch}, Low=Constant() {r=1})
   Forward=Constant() {r=1}
  )
  n36 = MaterialFunctionCall() {materialFunction=&ScreenResolution}
  n37 = ComponentMask(None=@n36.Visible Resolution) {r=true}
  n38 = ComponentMask(None=@n36.Visible Resolution) {g=true}
  n39 = Divide(
   A=ScalarParameter() {defaultValue=1.33, parameterName=ScreenRes_Mininum}
   B=ScalarParameter() {defaultValue=0.75, parameterName=SizeIncreaser_NintendoSwitch}
  )
  n40 = ShadingPathSwitch(Default=@n39, Forward=@n39)
  n41 = ScalarParameter() {defaultValue=2, parameterName=ScreenRes_Max}
  n42 = Subtract(A=Clamp(None=Divide(A=@n37, B=@n38), Min=@n40, Max=@n41) {maxDefault=2, minDefault=1.33}, B=@n40)
  n43 = Subtract(A=@n41, B=@n40)
  n44 = OneMinus(None=Saturate(None=Divide(A=@n42, B=@n43)))
  Reroute(None=LinearInterpolate(A=VectorParameter() {parameterName=TestOffset, defaultValue=<rgba 0 0 -500000 1>}.RGB, B=Multiply(A=@n35, B=@n44), Alpha=Clamp(None=@n32)))
 ]
}
