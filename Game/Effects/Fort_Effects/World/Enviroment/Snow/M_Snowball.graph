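# Masked, lit snowball material for mesh and Niagara mesh particles, built on the shared winter snow function.
# A powered UV-V gradient flattens the blended snow normals and masks noise from a panned snow texture into the dithered opacity, scaled by particle alpha.
# The same panner, under a separate gradient, drives displacement along vertex normals, with dynamic parameters controlling amount and speed.
# An optional tinted emissive passes through the shared FX adjustment.
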Material /Game/Effects/Fort_Effects/World/Enviroment/Snow/M_Snowball
graph {
 properties {
  bUsedWithMeshParticles=true
  bUsedWithNiagaraMeshParticles=true
  blendMode=BLEND_Masked
  pixelDepthOffsetMode=PDOM_Legacy
  usedShadingModels=Default Lit
 }
 nodes [
  n1 = MaterialFunctionCall!(MaterialLayers, UseVertexBlend) {materialFunction=/Game/Environments/Sets/Winter/Terrain/MF_SharedSnow}
  n2 = Constant3Vector() {constant=<rgba 0 0 1 0>}
  n3 = TextureCoordinate()
  n4 = OneMinus(None=Clamp(None=ComponentMask(None=@n3) {g=true}))
  n5 = Power(Base=@n4, Exp=ScalarParameter() {defaultValue=10, parameterName=VerticalOffset})
  LinearInterpolate(
   A=MaterialFunctionCall(
    BaseNormal=TextureSample() {samplerType=SAMPLERTYPE_Normal, texture=/Game/Environments/Sets/Winter/Terrain/T_Snow01_N}.RGB
    AdditionalNormal=@n1.OutputNormals
   ) {materialFunction=&BlendAngleCorrectedNormals}.Result
   B=@n2
   Alpha=@n5
  )
  n6 = VectorParameter() {parameterName=EmissiveColor, defaultValue=<rgba 1 1 1 1>}
  MaterialFunctionCall(Emissive=LinearInterpolate(B=Multiply(A=@n6.RGB, B=@n1.OutputRegularColor), Alpha=ScalarParameter() {parameterName=EmissmiveAmount})) {materialFunction=&MF_FX_EmissiveAdjustments}
  n7 = ParticleColor()
  n8 = Multiply(
   A=Multiply(A=ScalarParameter() {defaultValue=5, parameterName=Opacity _Scale}, B=@n7.A)
   B=LinearInterpolate(A=@n1.OutputZMask, B=@n2, Alpha=@n5)
  )
  n9 = Time()
  n10 = Multiply(
   A=ScalarParameter() {defaultValue=1, parameterName=WPO_PannerSpeedScale}
   B=DynamicParameter() {defaultValue=<rgba 0.2 1 1 1>, paramNames [WPO_Controller, WPO_PanningSpeedScale, Param3, Param4]}.WPO_PanningSpeedScale
  )
  n11 = TextureSample(UVs=Panner(Time=Multiply(A=@n9, B=@n10)) {speedY=1}) {texture=/Game/Environments/Sets/Winter/Terrain/T_Snow01_M}
  n12 = VectorNoise(World Position=@n11.RGB) {bTiling=true, noiseFunction=VNF_VectorALU}
  MaterialFunctionCall(Alpha Threshold=Multiply(A=LinearInterpolate(A=@n8, B=@n7.A, Alpha=@n5), B=LinearInterpolate(B=@n12, Alpha=@n5) {constA=1})) {materialFunction=&DitherTemporalAA_Fortnite}
  n13 = ScalarParameter() {defaultValue=50, parameterName=WPO_Max_Distance}
  n14 = LinearInterpolate(B=@n11.R, Alpha=Power(Base=@n4, Exp=ScalarParameter() {defaultValue=5, parameterName=VerticalOffset_WPO})) {constA=1}
  Multiply(
   A=Multiply(A=LinearInterpolate(A=@n13, Alpha=@n14) {constA=1, constB=0}, B=VertexNormalWS())
   B=Multiply(
    A=ScalarParameter() {defaultValue=1, parameterName=WPO_Controller_DEBUG}
    B=DynamicParameter() {defaultValue=<rgba 0.2 1 1 1>, paramNames [WPO_Controller, WPO_PanningSpeedScale, Param3, Param4]}.WPO_Controller
   )
  )
  Reroute(None=@n1.OutputSpec)
  Reroute(None=@n1.OutputRoughness)
 ]
}
